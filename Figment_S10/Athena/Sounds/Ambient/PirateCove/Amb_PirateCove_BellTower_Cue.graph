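# Sound cue for the church bell ring ambient at the Pirate Cove bell tower in Figment S10 Athena.
# A single wave player plays the midnight church bell wave from the town exterior ambient set, with no looping or randomization nodes.
# It uses the outdoor ambient sound class and overrides attenuation for a long falloff, wide stereo spread and a fixed manual reverb send.
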SoundCue /Figment_S10/Athena/Sounds/Ambient/PirateCove/Amb_PirateCove_BellTower_Cue
root=WavePlayer_0
nodes [{id=WavePlayer_0, properties {soundWaveAssetPtr=/Figment_S10/Sounds/Fort_Ambient_NonLoop/Town_Exterior/church_bell_midnight_ring}}]
properties {
 bOverrideAttenuation=true
 maxDistance=10000
 soundClassObject=&Fort_Ambient_Outdoors_Athena
 attenuationOverrides {falloffDistance=10000, manualReverbSendLevel=50, reverbSendMethod=Manual, stereoSpread=1024, attenuationShapeExtents {x=0}}
}
