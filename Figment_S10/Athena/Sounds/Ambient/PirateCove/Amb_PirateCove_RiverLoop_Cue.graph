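# Sound cue for the looping river water ambience in Pirate Cove, assigned to the outdoor ambient sound class.
# A looping water-edge wave feeds an Attenuation node, then a Mixer whose second input is empty.
# The node's attenuation is non-spatialized, with a custom curve reaching silence a fifth of the way through its falloff.
# The cue also carries its own attenuation override alongside a shared river attenuation asset reference.
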SoundCue /Figment_S10/Athena/Sounds/Ambient/PirateCove/Amb_PirateCove_RiverLoop_Cue
root=Mixer_0
nodes [
 {
  id=Attenuation_0
  children [WavePlayer_2]
  properties {
   bOverrideAttenuation=true
   attenuationOverrides {
    bSpatialize=false
    distanceAlgorithm=Custom
    falloffDistance=5500
    attenuationShapeExtents {x=0}
    customAttenuationCurve {editorCurveData {keys [<at 0 1>,<at 0.2 0>]}}
   }
  }
 }
 {id=Mixer_0, children [Attenuation_0, null], properties {inputVolume [1,1]}}
 {id=WavePlayer_2, properties {bLooping=true, soundWaveAssetPtr=&/Amb_PirateCove_WaterEdge_01}}
]
properties {
 attenuationSettings=/Game/Athena/Sounds/Ambient/Fort_Amb_Water_River_Att
 bOverrideAttenuation=true
 maxDistance=3925
 soundClassObject=&Fort_Ambient_Outdoors_Athena
 volumeMultiplier=1.2
 attenuationOverrides {falloffDistance=3000, nonSpatializedRadiusStart=1100, stereoSpread=1000, attenuationShapeExtents {x=925}}
 concurrencyOverrides {maxCount=1, resolutionRule=PreventNew}
}
